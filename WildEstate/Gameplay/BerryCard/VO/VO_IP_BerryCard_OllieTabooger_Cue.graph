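# Sound cue for the BerryCard "OllieTabooger" voice-over, a scripted prank phone call between BuzzCool and QueenTruth.
# A mixer starts a phone pickup sound alongside a fixed short delay, which leads into a concatenator alternating the two speakers' lines and ending on a hangup.
# One shared silence wave is reused between every clip to space the dialogue.
# Playback uses the narrator sound class, berry VO attenuation and per-owner concurrency that blocks overlapping plays.
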SoundCue /WildEstate/Gameplay/BerryCard/VO/VO_IP_BerryCard_OllieTabooger_Cue
root=Mixer_0
nodes [
 {
  id=Concatenator_1
  children [
   WavePlayer_5
   WavePlayer_0
   WavePlayer_9
   WavePlayer_0
   WavePlayer_6
   WavePlayer_0
   WavePlayer_10
   WavePlayer_0
   WavePlayer_7
   WavePlayer_0
   WavePlayer_11
   WavePlayer_0
   WavePlayer_8
   WavePlayer_0
   WavePlayer_12
   WavePlayer_0
   WavePlayer_3
  ]
  properties {inputVolume [1,1,1,1,1,1,1,1,1,1,1,1,1,1,1,1,1]}
 }
 {id=Delay_0, children [Concatenator_1], properties {delayMax=0.6, delayMin=0.6}}
 {id=Mixer_0, children [WavePlayer_2,Delay_0], properties {inputVolume [1,1]}}
 {id=WavePlayer_0, properties {soundWaveAssetPtr=/Game/Sounds/Fort_Human_VO_BR/NPCs/Shared/Point3Second_Silence}}
 {id=WavePlayer_10, properties {soundWaveAssetPtr=/WildEstate/Gameplay/BerryCard/VO/Waves/vo_IP_PrankCall_QueenTruth_051}}
 {id=WavePlayer_11, properties {soundWaveAssetPtr=/WildEstate/Gameplay/BerryCard/VO/Waves/vo_IP_PrankCall_QueenTruth_052}}
 {id=WavePlayer_12, properties {soundWaveAssetPtr=/WildEstate/Gameplay/BerryCard/VO/Waves/vo_IP_PrankCall_QueenTruth_053}}
 {id=WavePlayer_2, properties {soundWaveAssetPtr=/WildEstate/Gameplay/BerryCard/Audio/Phone_Pickup_04}}
 {id=WavePlayer_3, properties {soundWaveAssetPtr=/WildEstate/Gameplay/BerryCard/Audio/Phone_Hangup_04}}
 {id=WavePlayer_5, properties {soundWaveAssetPtr=/WildEstate/Gameplay/BerryCard/VO/Waves/vo_IP_PrankCall_BuzzCool_050}}
 {id=WavePlayer_6, properties {soundWaveAssetPtr=/WildEstate/Gameplay/BerryCard/VO/Waves/vo_IP_PrankCall_BuzzCool_051}}
 {id=WavePlayer_7, properties {soundWaveAssetPtr=/WildEstate/Gameplay/BerryCard/VO/Waves/vo_IP_PrankCall_BuzzCool_052}}
 {id=WavePlayer_8, properties {soundWaveAssetPtr=/WildEstate/Gameplay/BerryCard/VO/Waves/vo_IP_PrankCall_BuzzCool_053}}
 {id=WavePlayer_9, properties {soundWaveAssetPtr=/WildEstate/Gameplay/BerryCard/VO/Waves/vo_IP_PrankCall_QueenTruth_050}}
]
properties {
 attenuationSettings=/WildEstate/Gameplay/BerryCard/Audio/ATT_BR_Fort_Berry_VO
 maxDistance=2300
 soundClassObject=/Game/Sounds/FortSoundClassesAndModes/Fort_VO_Narrator
 volumeMultiplier=1.6
 concurrencyOverrides {bLimitToOwner=true, maxCount=1, resolutionRule=PreventNew}
}
