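# Sound cue that plays the "SeymourButz" prank phone call voice sequence for the BerryCard gameplay feature.
# A Mixer starts the phone pickup sound alongside a briefly delayed Concatenator, which chains DuckCoast, QueenTruth and DuckCoast lines with a shared silence clip after each, then a hangup.
# It uses the narrator VO class with BerryCard attenuation and owner-limited, single-instance concurrency that prevents new plays while one is active.
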SoundCue /WildEstate/Gameplay/BerryCard/VO/VO_IP_BerryCard_SeymourButz_Cue
root=Mixer_0
nodes [
 {
  id=Concatenator_0
  children [WavePlayer_1,WavePlayer_0,WavePlayer_7,WavePlayer_0,WavePlayer_6,WavePlayer_0,WavePlayer_3]
  properties {inputVolume [1,1,1,1,1,1,1]}
 }
 {id=Delay_0, children [Concatenator_0], properties {delayMax=0.6, delayMin=0.6}}
 {id=Mixer_0, children [WavePlayer_2,Delay_0], properties {inputVolume [1,1]}}
 {id=WavePlayer_0, properties {soundWaveAssetPtr=/Game/Sounds/Fort_Human_VO_BR/NPCs/Shared/Point3Second_Silence}}
 {id=WavePlayer_1, properties {soundWaveAssetPtr=/WildEstate/Gameplay/BerryCard/VO/Waves/vo_IP_PrankCall_DuckCoast_040}}
 {id=WavePlayer_2, properties {soundWaveAssetPtr=/WildEstate/Gameplay/BerryCard/Audio/Phone_Pickup_01}}
 {id=WavePlayer_3, properties {soundWaveAssetPtr=/WildEstate/Gameplay/BerryCard/Audio/Phone_Hangup_01}}
 {id=WavePlayer_6, properties {soundWaveAssetPtr=/WildEstate/Gameplay/BerryCard/VO/Waves/vo_IP_PrankCall_DuckCoast_061}}
 {id=WavePlayer_7, properties {soundWaveAssetPtr=/WildEstate/Gameplay/BerryCard/VO/Waves/vo_IP_PrankCall_QueenTruth_060}}
]
properties {
 attenuationSettings=/WildEstate/Gameplay/BerryCard/Audio/ATT_BR_Fort_Berry_VO
 maxDistance=2300
 soundClassObject=/Game/Sounds/FortSoundClassesAndModes/Fort_VO_Narrator
 volumeMultiplier=1.6
 concurrencyOverrides {bLimitToOwner=true, maxCount=1, resolutionRule=PreventNew}
}
